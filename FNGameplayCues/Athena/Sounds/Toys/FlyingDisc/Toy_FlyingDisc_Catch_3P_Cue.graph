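# Third-person sound cue for catching the Flying Disc toy.
# A mixer layers one of three random catch sounds, passed through a modulator for variation, with a quieter random score "pop" borrowed from LTM audio.
# The pop starts after a short fixed delay, so it follows the catch.
# Playback uses toy-hit attenuation and is routed to the music emote sound class.
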SoundCue /FNGameplayCues/Athena/Sounds/Toys/FlyingDisc/Toy_FlyingDisc_Catch_3P_Cue
root=Mixer_0
nodes [
 {id=Delay_0, children [Random_2], properties {delayMax=0.35, delayMin=0.35}}
 {id=Mixer_0, children [Delay_0,Modulator_0], properties {inputVolume [0.5,1]}}
 {id=Modulator_0, children [Random_1], properties {volumeMax=1}}
 {id=Random_1, children [WavePlayer_2,WavePlayer_3,WavePlayer_4], properties {weights [1,1,1]}}
 {id=Random_2, children [WavePlayer_5,WavePlayer_6,WavePlayer_7], properties {weights [1,1,1]}}
 {id=WavePlayer_2, properties {soundWaveAssetPtr=&/Athena_Toy_FlyingDisc_Catch_01}}
 {id=WavePlayer_3, properties {soundWaveAssetPtr=&/Athena_Toy_FlyingDisc_Catch_02}}
 {id=WavePlayer_4, properties {soundWaveAssetPtr=&/Athena_Toy_FlyingDisc_Catch_03}}
 {id=WavePlayer_5, properties {soundWaveAssetPtr=/Game/Athena/Sounds/LTMs/Score/ScoreLTM_InGamePop_Little_01}}
 {id=WavePlayer_6, properties {soundWaveAssetPtr=/Game/Athena/Sounds/LTMs/Score/ScoreLTM_InGamePop_Little_02}}
 {id=WavePlayer_7, properties {soundWaveAssetPtr=/Game/Athena/Sounds/LTMs/Score/ScoreLTM_InGamePop_Little_03}}
]
properties {
 attenuationSettings=/Game/Sounds/Attenuation/Fort_ToyHit_Att
 maxDistance=2800
 soundClassObject=&Fort_Music_Emote
}
